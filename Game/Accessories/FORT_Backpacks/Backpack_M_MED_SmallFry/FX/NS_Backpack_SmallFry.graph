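# Niagara system for the SmallFry backpack that renders two layered pairs of translucent, fluttering mesh wings.
# Four mesh emitters (A and B per side) pin particles to two bone or socket points on the User skeletal mesh.
# Spawn rate rises and falls over each emitter loop.
# Each short-lived particle takes a flap angle from a sawtooth of emitter age at spawn, so overlapping copies form the flutter.
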NiagaraSystem /Game/Accessories/FORT_Backpacks/Backpack_M_MED_SmallFry/FX/NS_Backpack_SmallFry
dependencies {usedDataInterfaces [&NiagaraDataInterfaceCurve], usedDynamicInputs [], usedModules [&SystemState]}
emitters {
 Left_Wing_A001 {
  emitterUpdateScript [
   &EmitterState()
   &SpawnRate(
    SpawnRate=&FloatFromCurve(
     FloatCurve=DataInterface() {propertyValues {Curve {keys [
         <at 0 0 RCIM_Cubic>
         {arriveTangent=-328.9862, interpMode=RCIM_Cubic, leaveTangent=-328.9862, tangentMode=RCTM_User, time=0.35, value=60}
         {arriveTangent=-77.5329, interpMode=RCIM_Cubic, leaveTangent=-77.5429, tangentMode=RCTM_User, time=0.8, value=0}
        ]}}}
     CurveIndex=$Emitter.NormalizedLoopAge
    )
    Use Spawn Probability=false
   )
  ]
  particleSpawnScript [
   /SetVariables(Particles.Random01:NiagaraFloat=&RandomRangeFloat(Minimum=0, Maximum=1)) {isSetParameters=true, module=SetVariables_#1}
   /SetVariables(Transient.WingFlapRangeTopRange=60) {isSetParameters=true, module=SetVariables_#2}
   /SetVariables(Transient.WingFlapRangeBottomRange=-10) {isSetParameters=true, module=SetVariables_#3}
   &InitializeParticle(
    Color=&MakeLinearColorFromVectorAndFloat("Vector (RGB)"=&MakeVectorFromLinearColorRGB(Color=<rgba 1 0.7 0.6 1>), "Float (Alpha)"=0.2)
    Lifetime=0.1
    Mesh Scale=&Multiply_Vector(A=&MakeVector(X=1.5, Y=1.5, Z=1), B=&VectorFromFloat(Value=1))
    Mesh Scale Mode=::NewEnumerator4
    Sprite Size Mode=::NewEnumerator3
    Uniform Sprite Size=20
    UsePositionOffset=false
    Write Material Random=true
   )
   &InitialMeshOrientation(Orientation Axis=<0 1 0>, Orientation Vector=<0.3 0 1>, Rotation=<0.5 0 0>, Use Orientation Vector=true, Use Rotation Vector=true)
   /SetVariables(Particles.MeshOrientation=&MultiplyQuaternion(
     Quaternion A=$Particles.MeshOrientation
     Quaternion B=&MultiplyQuaternion(
      Quaternion A=&MakeQuatFromAxisAngle(
       Axis=<1 0 0>
       AngleInDegrees=&Lerp_Float(A=0, B=0, Alpha=&Frac_Float(Float=&Multiply_Float(A=$Emitter.Age, B=10)))
      )
      Quaternion B=&MultiplyQuaternion(
       Quaternion A=&MakeQuatFromAxisAngle(
        Axis=<0 1 0>
        AngleInDegrees=&Lerp_Float(
         A=$Transient.WingFlapRangeBottomRange
         B=$Transient.WingFlapRangeTopRange
         Alpha=&Frac_Float(Float=&Multiply_Float(A=$Emitter.Age, B=15.5))
        )
       )
       Quaternion B=&MakeQuatFromAxisAngle(
        Axis=<0 0 1>
        AngleInDegrees=&Lerp_Float(A=0, B=0, Alpha=&Frac_Float(Float=&Multiply_Float(A=$Emitter.Age, B=10)))
       )
      )
     )
    )) {isSetParameters=true, module=SetVariables_#4}
  ]
  particleUpdateScript [
   &SkeletalMeshLocation(
    Inherit Velocity=false
    Interpolate Mesh Sampling=false
    MaskSpawnGroup=false
    Mesh Sampling Type=::NewEnumerator5
    Skeletal Mesh=$User.SkeletalMesh
    Socket Sampling Mode=::NewEnumerator5
   )
   &ParticleState()
   &UpdateMeshOrientation(
    Rotation Rate=0
    Rotation Vector=&RotateVectorByQuaternion(
     Vector To Rotate=&ReturnMeshOrientationAxis(
      Mesh Orientation Axis=::XAxis
      Mesh Orientation Quat=&MultiplyQuaternion(
       Quaternion A=&OrientMeshToVector(Look At Direction=<1 0 -1>, Reference Axis Vector=<1 0 0>)
       Quaternion B=<0 0 0 1>
      )
     )
     Quaternion=<0 0 0 1>
    )
   )
  ]
  rendererClasses [&NiagaraMeshRendererProperties]
  renderers [{rendererClass=&NiagaraMeshRendererProperties, rendererIndex=0}]
 }
 Left_Wing_B001 {
  emitterUpdateScript [
   &EmitterState()
   &SpawnRate(
    SpawnRate=&FloatFromCurve(
     FloatCurve=DataInterface() {propertyValues {Curve {keys [
         {arriveTangent=580.4644, interpMode=RCIM_Cubic, leaveTangent=580.4642, tangentMode=RCTM_User, time=0, value=0}
         {arriveTangent=-328.9862, interpMode=RCIM_Cubic, leaveTangent=-328.9862, tangentMode=RCTM_User, time=0.35, value=50}
         {arriveTangent=-77.5329, interpMode=RCIM_Cubic, leaveTangent=-77.5429, tangentMode=RCTM_User, time=0.8, value=0}
        ]}}}
     CurveIndex=$Emitter.NormalizedLoopAge
    )
    Use Spawn Probability=false
   )
  ]
  particleSpawnScript [
   /SetVariables(Particles.Random01:NiagaraFloat=&RandomRangeFloat(Minimum=0, Maximum=1)) {isSetParameters=true, module=SetVariables_#1}
   /SetVariables(Transient.WingFlapRangeTopRange=40) {isSetParameters=true, module=SetVariables_#5}
   /SetVariables(Transient.WingFlapRangeBottomRange=-20) {isSetParameters=true, module=SetVariables_#6}
   &InitializeParticle(
    Color=&MakeLinearColorFromVectorAndFloat("Vector (RGB)"=&MakeVectorFromLinearColorRGB(Color=<rgba 1 0.7 0.6 1>), "Float (Alpha)"=0.15)
    Lifetime=0.1
    Mesh Scale=&Multiply_Vector(
     A=&MakeVector(
      X=1
      Y=1.5
      Z=&Lerp_Float(
       A=1
       B=1
       Alpha=/Niagara/DynamicInputs/Integer/V2/Round_Float(Float:NiagaraFloat=&Frac_Float(Float=&Multiply_Float(A=$Emitter.Age, B=15.5)))
      )
     )
     B=&VectorFromFloat(Value=1)
    )
    Mesh Scale Mode=::NewEnumerator4
    Sprite Size Mode=::NewEnumerator3
    Uniform Sprite Size=20
    UsePositionOffset=false
    Write Material Random=true
   )
   &InitialMeshOrientation(Orientation Axis=<0 1 0>, Orientation Vector=<0.3 0 1>, Rotation=<0.5 0 -0.05>, Use Orientation Vector=true, Use Rotation Vector=true)
   /SetVariables(Particles.MeshOrientation=&MultiplyQuaternion(
     Quaternion A=$Particles.MeshOrientation
     Quaternion B=&MultiplyQuaternion(
      Quaternion A=&MakeQuatFromAxisAngle(
       Axis=<1 0 0>
       AngleInDegrees=&Lerp_Float(A=0, B=0, Alpha=&Frac_Float(Float=&Multiply_Float(A=$Emitter.Age, B=10)))
      )
      Quaternion B=&MultiplyQuaternion(
       Quaternion A=&MakeQuatFromAxisAngle(
        Axis=<0 1 0>
        AngleInDegrees=&Lerp_Float(
         A=$Transient.WingFlapRangeBottomRange
         B=$Transient.WingFlapRangeTopRange
         Alpha=&Frac_Float(Float=&Multiply_Float(A=$Emitter.Age, B=15.5))
        )
       )
       Quaternion B=&MakeQuatFromAxisAngle(
        Axis=<0 0 1>
        AngleInDegrees=&Lerp_Float(A=0, B=0, Alpha=&Frac_Float(Float=&Multiply_Float(A=$Emitter.Age, B=10)))
       )
      )
     )
    )) {isSetParameters=true, module=SetVariables_#4}
  ]
  particleUpdateScript [
   &SkeletalMeshLocation(
    Inherit Velocity=false
    Interpolate Mesh Sampling=false
    MaskSpawnGroup=false
    Mesh Sampling Type=::NewEnumerator5
    Skeletal Mesh=$User.SkeletalMesh
    Socket Sampling Mode=::NewEnumerator5
   )
   &ParticleState()
   &UpdateMeshOrientation(Rotation Rate=0, Rotation Vector=<1 0 0>)
  ]
  rendererClasses [&NiagaraMeshRendererProperties]
  renderers [{rendererClass=&NiagaraMeshRendererProperties, rendererIndex=0}]
 }
 Right_Wing_A001 {
  emitterUpdateScript [
   &EmitterState()
   &SpawnRate(
    SpawnRate=&FloatFromCurve(
     FloatCurve=DataInterface() {propertyValues {Curve {keys [
         <at 0 0 RCIM_Cubic>
         {arriveTangent=-328.9862, interpMode=RCIM_Cubic, leaveTangent=-328.9862, tangentMode=RCTM_User, time=0.35, value=60}
         {arriveTangent=-77.5329, interpMode=RCIM_Cubic, leaveTangent=-77.5429, tangentMode=RCTM_User, time=0.8, value=0}
        ]}}}
     CurveIndex=$Emitter.NormalizedLoopAge
    )
    Use Spawn Probability=false
   )
  ]
  particleSpawnScript [
   /SetVariables(Particles.Random01:NiagaraFloat=&RandomRangeFloat(Minimum=0, Maximum=1)) {isSetParameters=true, module=SetVariables_#1}
   /SetVariables(Transient.WingFlapRangeTopRange=-10) {isSetParameters=true, module=SetVariables_#2}
   /SetVariables(Transient.WingFlapRangeBottomRange=60) {isSetParameters=true, module=SetVariables_#3}
   &InitializeParticle(
    Color=&MakeLinearColorFromVectorAndFloat("Vector (RGB)"=&MakeVectorFromLinearColorRGB(Color=<rgba 1 0.7 0.6 1>), "Float (Alpha)"=0.2)
    Lifetime=0.1
    Mesh Scale=&Multiply_Vector(A=&MakeVector(X=1.5, Y=1.5, Z=1), B=&VectorFromFloat(Value=-1))
    Mesh Scale Mode=::NewEnumerator4
    Sprite Size Mode=::NewEnumerator3
    Uniform Sprite Size=20
    UsePositionOffset=false
    Write Material Random=true
   )
   &InitialMeshOrientation(Orientation Axis=<0 1 0>, Orientation Vector=<0.25 0 1>, Use Orientation Vector=true, Use Rotation Vector=true)
   /SetVariables(Particles.MeshOrientation=&MultiplyQuaternion(
     Quaternion A=$Particles.MeshOrientation
     Quaternion B=&MultiplyQuaternion(
      Quaternion A=&MakeQuatFromAxisAngle(
       Axis=<1 0 0>
       AngleInDegrees=&Lerp_Float(A=0, B=0, Alpha=&Frac_Float(Float=&Multiply_Float(A=$Emitter.Age, B=10)))
      )
      Quaternion B=&MultiplyQuaternion(
       Quaternion A=&MakeQuatFromAxisAngle(
        Axis=<0 1 0>
        AngleInDegrees=&Lerp_Float(
         A=$Transient.WingFlapRangeBottomRange
         B=$Transient.WingFlapRangeTopRange
         Alpha=&Frac_Float(Float=&Multiply_Float(A=$Emitter.Age, B=15.5))
        )
       )
       Quaternion B=&MakeQuatFromAxisAngle(
        Axis=<0 0 1>
        AngleInDegrees=&Lerp_Float(A=0, B=0, Alpha=&Frac_Float(Float=&Multiply_Float(A=$Emitter.Age, B=10)))
       )
      )
     )
    )) {isSetParameters=true, module=SetVariables_#4}
  ]
  particleUpdateScript [
   &SkeletalMeshLocation(
    Bone / Socket Index=1
    Inherit Velocity=false
    Interpolate Mesh Sampling=false
    MaskSpawnGroup=false
    Mesh Sampling Type=::NewEnumerator5
    Skeletal Mesh=$User.SkeletalMesh
    Socket Sampling Mode=::NewEnumerator5
   )
   &ParticleState()
  ]
  rendererClasses [&NiagaraMeshRendererProperties]
  renderers [{rendererClass=&NiagaraMeshRendererProperties, rendererIndex=0}]
 }
 Right_Wing_B001 {
  emitterUpdateScript [
   &EmitterState()
   &SpawnRate(
    SpawnRate=&FloatFromCurve(
     FloatCurve=DataInterface() {propertyValues {Curve {keys [
         {arriveTangent=580.4644, interpMode=RCIM_Cubic, leaveTangent=580.4642, tangentMode=RCTM_User, time=0, value=0}
         {arriveTangent=-328.9862, interpMode=RCIM_Cubic, leaveTangent=-328.9862, tangentMode=RCTM_User, time=0.35, value=50}
         {arriveTangent=-77.5329, interpMode=RCIM_Cubic, leaveTangent=-77.5429, tangentMode=RCTM_User, time=0.8, value=0}
        ]}}}
     CurveIndex=$Emitter.NormalizedLoopAge
    )
    Use Spawn Probability=false
   )
  ]
  particleSpawnScript [
   /SetVariables(Particles.Random01:NiagaraFloat=&RandomRangeFloat(Minimum=0, Maximum=1)) {isSetParameters=true, module=SetVariables_#1}
   /SetVariables(Transient.WingFlapRangeTopRange=10) {isSetParameters=true, module=SetVariables_#5}
   /SetVariables(Transient.WingFlapRangeBottomRange=65) {isSetParameters=true, module=SetVariables_#6}
   &InitializeParticle(
    Color=&MakeLinearColorFromVectorAndFloat("Vector (RGB)"=&MakeVectorFromLinearColorRGB(Color=<rgba 1 0.7 0.6 1>), "Float (Alpha)"=0.15)
    Lifetime=0.1
    Mesh Scale=&Multiply_Vector(
     A=&MakeVector(
      X=1
      Y=1.5
      Z=&Lerp_Float(
       A=1
       B=1
       Alpha=&Round_Float(Float:NiagaraFloat=&Frac_Float(Float=&Multiply_Float(A=$Emitter.Age, B=15.5)))
      )
     )
     B=&VectorFromFloat(Value=1)
    )
    Mesh Scale Mode=::NewEnumerator4
    Sprite Size Mode=::NewEnumerator3
    Uniform Sprite Size=20
    UsePositionOffset=false
    Write Material Random=true
   )
   &InitialMeshOrientation(Orientation Axis=<0 -1 0>, Orientation Vector=<-1 1 0>, Rotation=<-0.25 0 0>, Use Orientation Vector=true, Use Rotation Vector=true)
   /SetVariables(Particles.MeshOrientation=&MultiplyQuaternion(
     Quaternion A=$Particles.MeshOrientation
     Quaternion B=&MultiplyQuaternion(
      Quaternion A=&MakeQuatFromAxisAngle(
       Axis=<1 0 0>
       AngleInDegrees=&Lerp_Float(A=0, B=0, Alpha=&Frac_Float(Float=&Multiply_Float(A=$Emitter.Age, B=10)))
      )
      Quaternion B=&MultiplyQuaternion(
       Quaternion A=&MakeQuatFromAxisAngle(
        Axis=<0 1 0>
        AngleInDegrees=&Lerp_Float(
         A=$Transient.WingFlapRangeBottomRange
         B=$Transient.WingFlapRangeTopRange
         Alpha=&Frac_Float(Float=&Multiply_Float(A=$Emitter.Age, B=15.5))
        )
       )
       Quaternion B=&MakeQuatFromAxisAngle(
        Axis=<0 0 1>
        AngleInDegrees=&Lerp_Float(A=0, B=0, Alpha=&Frac_Float(Float=&Multiply_Float(A=$Emitter.Age, B=10)))
       )
      )
     )
    )) {isSetParameters=true, module=SetVariables_#4}
  ]
  particleUpdateScript [
   &SkeletalMeshLocation(
    Bone / Socket Index=1
    Inherit Velocity=false
    Interpolate Mesh Sampling=false
    MaskSpawnGroup=false
    Mesh Sampling Type=::NewEnumerator5
    Skeletal Mesh=$User.SkeletalMesh
    Socket Sampling Mode=::NewEnumerator5
   )
   &ParticleState()
  ]
  rendererClasses [&NiagaraMeshRendererProperties]
  renderers [{rendererClass=&NiagaraMeshRendererProperties, rendererIndex=0}]
 }
}
summary {
 systemName=NS_Backpack_SmallFry
 emitters [
  {emitterName=Left_Wing_A001, rendererClasses [&NiagaraMeshRendererProperties]}
  {emitterName=Left_Wing_B001, rendererClasses [&NiagaraMeshRendererProperties]}
  {emitterName=Right_Wing_B001, rendererClasses [&NiagaraMeshRendererProperties]}
  {emitterName=Right_Wing_A001, rendererClasses [&NiagaraMeshRendererProperties]}
 ]
 userVariables [
  User.PhysMatIndex {type=NiagaraInt32}
  User.SkeletalMesh {
   type=NiagaraDataInterfaceSkeletalMesh
   defaultValue {
    struct=NiagaraExt_VariableValue_DataInterface
    value {dataInterface=&self:NiagaraDataInterfaceSkeletalMesh_0}
   }
  }
 ]
}
